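# Translucent unlit material for Cube Queen pickaxe particle effects, serving either a cube look or a deploy trail.
# The cube look blends a screen-aligned nebula texture with two panning noises colored through a curve-atlas gradient.
# The trail uses a three-color gradient, scrolling texture, dynamic-parameter-driven sweep mask and SplineThicken width offset.
# The Switch_to_the_Deploy_Trail scalar lerps emissive, opacity and vertex offset between looks, with particle color tinting both.
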Material /BRCosmetics/Weapons/FORT_Melee/Pickaxe_Cube_Queen_Female/FX/M_CubeQueen_Cube
graph {
 properties {
  bUsedWithNiagaraMeshParticles=true
  blendMode=BLEND_Translucent
  pixelDepthOffsetMode=PDOM_Legacy
  shadingModel=MSM_Unlit
  usedShadingModels=Unlit
 }
 nodes [
  n1 = Panner(
   Coordinate=TextureCoordinate() {uTiling=0.02, vTiling=0.03}
   Time=Constant() {r=1}
   Speed=ComponentMask(None=VectorParameter() {group=Cube Queen Emissive, parameterName=UV_Offset, defaultValue=<rgba 0.5 0.3 0 1>}.RGB) {g=true, r=true}
  ) {speedX=0.02, speedY=0.02}
  n2 = TextureSampleParameter2D(UVs=@n1) {
   group=Cube Queen Emissive
   parameterName=NoiseB
   samplerType=SAMPLERTYPE_LinearGrayscale
   texture=/Game/Athena/Prototype/Terrain/T_CloudCoverage
  }
  n3 = TextureSampleParameter2D(UVs=@n1) {
   group=Cube Queen Emissive
   parameterName=NoiseA
   texture=/Game/CharClasses/Outlander/Gadgets/AntiMaterialCharge/Materials/T_AntiMaterialCharge_Dissolve_01
  }
  n4 = LinearInterpolate(
   A=Power(Base=@n2.R, Exp=ScalarParameter() {defaultValue=2, group=Cube Queen Emissive, parameterName=NoiseBPow})
   B=Power(Base=Multiply(A=@n3.R) {constB=2}, Exp=ScalarParameter() {defaultValue=2, group=Cube Queen Emissive, parameterName=NoiseAPow})
  ) {constAlpha=0.4}
  Add(A=@n4)
  n5 = ParticleColor()
  n6 = ParticleColor()
  n7 = TextureCoordinate() {uTiling=1.5}
  n8 = AppendVector(A=Constant(), B=DynamicParameter() {defaultValue=<rgba 0 1 1 1>, paramNames [Tex_Panner_offset, Width, Width_top, Random]}.Random)
  n9 = TextureSampleParameter2D(UVs=Panner(Coordinate=Add(A=@n7, B=@n8)) {speedX=-1.5}) {
   group=The_Deploy_Trail
   parameterName=Base_Trail_texture
   texture=/BRCosmetics/Weapons/FORT_Melee/Pickaxe_Tower_Sentinel_Female/FX/T_TowerSentinel_trail
  }
  n10 = Multiply(A=@n6.A, B=Power(Base=@n9.R) {constExponent=1})
  n11 = Constant() {r=0.7}
  n12 = Constant() {r=-0.8}
  n13 = DynamicParameter() {defaultValue=<rgba 0 1 1 1>, paramNames [Tex_Panner_offset, Width, Width_top, Random]}
  n14 = Add(
   A=Add(A=TextureCoordinate() {vTiling=0.5}, B=Constant2Vector() {g=0.35})
   B=AppendVector(A=LinearInterpolate(A=@n11, B=@n12, Alpha=@n13.Tex_Panner_offset), B=Constant())
  )
  n15 = Saturate(None=Multiply(A=TextureSample(UVs=@n14) {samplerSource=SSM_Clamp_WorldGroupSettings, texture=/Game/WaterAndWind/Textures/WeaponFireVectorMask}.R) {constB=2})
  n16 = Sine(None=MaterialFunctionCall(Float2=TextureCoordinate()) {materialFunction=&BreakOutFloat2Components}.R) {period=2}
  LinearInterpolate(
   A=@n5.A
   B=Multiply(A=Multiply(A=@n10, B=@n15), B=Power(Base=@n16))
   Alpha=ScalarParameter() {group=The_Deploy_Trail, parameterName=Switch_to_the_Deploy_Trail}
  )
  n17 = DynamicParameter() {defaultValue=<rgba 0 1 1 1>, paramNames [Tex_Panner_offset, Width, Width_top, Random]}
  LinearInterpolate(
   A=Constant()
   B=MaterialFunctionCall(WidthBase=@n17.Width, WidthTip=@n17.Width, Expand U or V UV Channel=StaticBool()) {materialFunction=&SplineThicken}.WorldPositionOffset
   Alpha=ScalarParameter() {group=The_Deploy_Trail, parameterName=Switch_to_the_Deploy_Trail}
  )
  MaterialFunctionCall!(
   CameraVector
   Texture=TextureObject() {texture=/Game/Characters/Player/Male/Medium/Bodies/M_Med_Soldier_04/Skins/Galaxy/Textures/T_FN_Nebula_Pink}
   Tiling=VectorParameter() {parameterName=Tex_Tiling, defaultValue=<rgba 10 10 10 0>}.RGB
   Offset
   AxisFadeContrast
  ) {materialFunction=/Engine/Functions/Engine_MaterialFunctions01/Texturing/TriplanarCameraVector}
  n18 = TextureSample(UVs=MaterialFunctionCall!(TextureResolution) {materialFunction=/Engine/Functions/Engine_MaterialFunctions02/Texturing/ScreenAlignedPixelToPixelUVs}.Result) {texture=&T_FN_Nebula_Pink}
  n19 = Constant()
  n20 = CurveAtlasRowParameter(CurveTime=Frac(None=@n4)) {
   atlas=/Game/Athena/LinearColorCurves/VFX_CurveAtlas
   curve=/Game/Weapons/FORT_Melee/Pickaxe_Cube_Queen_Female/FX/Cube_Queen_EmissiveCurve
   defaultValue=41
   group=Cube Queen Emissive
   parameterName=NoiseGradient
  }
  n21 = ScalarParameter() {defaultValue=0.5, group=Cube Queen Emissive, parameterName=NoiseEmissiveStrength}
  n22 = LinearInterpolate(
   A=Multiply(A=Desaturation(None=@n18.RGB, Fraction=@n19)) {constB=3}
   B=Desaturation(None=Multiply(A=@n20, B=@n21), Fraction=ScalarParameter() {group=Cube Queen Emissive, parameterName=Desaturation})
   Alpha=ScalarParameter() {defaultValue=0.3, group=Cube Queen Emissive, parameterName=Lerp_Cube_Tex}
  )
  n23 = TextureCoordinate()
  n24 = MaterialFunctionCall(
   A=VectorParameter() {group=The_Deploy_Trail, parameterName=Color1, defaultValue=<rgba 0 0.742 1 1>}.RGB
   B=VectorParameter() {group=The_Deploy_Trail, parameterName=Color2, defaultValue=<rgba 0 0.25 1 1>}.RGB
   C=VectorParameter() {group=The_Deploy_Trail, parameterName=Color3, defaultValue=<rgba 0 0.0145 1 1>}.RGB
   Alpha=Multiply(
    A=MaterialFunctionCall(Float2=@n23) {materialFunction=&BreakOutFloat2Components}.R
    B=ScalarParameter() {defaultValue=0.5, group=The_Deploy_Trail, parameterName=Gradient_intensity}
   ) {constB=0.5}
  ) {materialFunction=&3ColorBlend}
  MaterialFunctionCall(Emissive=LinearInterpolate(
    A=Multiply(A=@n5.RGB, B=@n22)
    B=Multiply(A=@n6.RGB, B=@n24.Result)
    Alpha=ScalarParameter() {group=The_Deploy_Trail, parameterName=Switch_to_the_Deploy_Trail}
   )) {materialFunction=&MF_FX_EmissiveAdjustments}
 ]
}
