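# Animation Blueprint for the Cube Queen (female) pickaxe, adding secondary physics motion to its crown and floating cube bones.
# Long chains of single-bone Anim Dynamics nodes apply spring-driven, mostly translation-only movement with per-bone alphas and limits, mostly without gravity.
# Several differently tuned chains cover the same cubes, simulating relative to the root, master bone or handle.
# At least two chains start from the cached 'CubesFloat' pose.
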AnimBlueprint /BRCosmetics/Weapons/FORT_Melee/Pickaxe_Cube_Queen_Female/Meshes/Cube_Queen_Female_Axe_AnimBP
target_skeleton=&/Cube_Queen_Female_Axe_Skeleton
graphs [
 AnimGraph {nodes [
   Anim Dynamics - Bone: dyn_crown(Alpha=0.15, ExternalForce="0,0,0") {
    angularDampingOverride=1
    angularSpringConstant=20000
    bAngularSpring=true
    bGravityOverrideInSimSpace=true
    bLinearSpring=true
    bOverrideAngularDamping=true
    bOverrideLinearDamping=true
    bUseGravityOverride=true
    gravityScale=0.35
    linearDampingOverride=1
    linearSpringConstant=120
    simulationSpace=RootRelative
    alphaBoolBlend {blendOption=Linear}
    boundBone {boneName=dyn_crown}
    componentAppliedLinearAccClamp=<3000 3000 3000>
    componentLinearAccScale=<0.3 0.3 0.8>
    componentLinearVelScale=<0.3 0.3 0.8>
    gravityOverride {z=-980}
    physicsBodyDefinitions [{
      sphereCollisionRadius=0
      boundBone {boneName=dyn_crown}
      boxExtents=<30 40 40>
      constraintSetup {angularLimitsMax=<0 0 0>, angularLimitsMin=<0 0 0>, angularTarget=<1 0 0>, linearAxesMax=<0 50 50>, linearAxesMin=<-50 -50 -50>}
      localJointOffset=<0 0 0>
     }]
   }
   n1 = Anim Dynamics - Bone: dyn_cube_l_in(Alpha=0.2, ExternalForce="0,0,0") {
    angularDampingOverride=0.9
    angularSpringConstant=20000
    bAngularSpring=true
    bGravityOverrideInSimSpace=true
    bLinearSpring=true
    bOverrideAngularDamping=true
    bOverrideLinearDamping=true
    gravityScale=0
    linearDampingOverride=0.995
    linearSpringConstant=150
    simulationSpace=BoneRelative
    alphaBoolBlend {blendOption=Linear}
    boundBone {boneName=dyn_cube_l_in}
    componentAppliedLinearAccClamp=<6000 6000 6000>
    componentLinearAccScale=<0.13 0.13 0.13>
    componentLinearVelScale=<0.9 0.9 0.9>
    gravityOverride {z=-980}
    physicsBodyDefinitions [{
      sphereCollisionRadius=0
      boundBone {boneName=dyn_cube_l_in}
      boxExtents=<5 5 5>
      constraintSetup {
       angularTargetAxis=AxisZ
       twistAxis=AxisZ
       angularLimitsMax=<0 0 0>
       angularLimitsMin=<0 0 0>
       angularTarget=<0 0 0>
       linearAxesMax=<1 3 2>
       linearAxesMin=<-1 -2 -3>
      }
      localJointOffset=<-0.1 -0.2 0>
     }]
    relativeSpaceBone {boneName=Melee_TwoHanded_Handle}
   }
   n2 = Anim Dynamics - Bone: dyn_cube_l_mid(ComponentPose=@n1, Alpha=0.35, ExternalForce="0,0,0") {
    angularDampingOverride=0.9
    angularSpringConstant=20000
    bAngularSpring=true
    bGravityOverrideInSimSpace=true
    bLinearSpring=true
    bOverrideAngularDamping=true
    bOverrideLinearDamping=true
    gravityScale=0
    linearDampingOverride=0.995
    linearSpringConstant=150
    simulationSpace=BoneRelative
    alphaBoolBlend {blendOption=Linear}
    boundBone {boneName=dyn_cube_l_mid}
    componentAppliedLinearAccClamp=<6000 6000 6000>
    componentLinearAccScale=<0.1 0.1 0.1>
    componentLinearVelScale=<0.9 0.9 0.9>
    gravityOverride {z=-980}
    physicsBodyDefinitions [{
      sphereCollisionRadius=0
      boundBone {boneName=dyn_cube_l_mid}
      boxExtents=<10 10 10>
      constraintSetup {
       angularTargetAxis=AxisZ
       twistAxis=AxisZ
       angularLimitsMax=<0 0 0>
       angularLimitsMin=<0 0 0>
       angularTarget=<0 0 0>
       linearAxesMax=<1 3 2>
       linearAxesMin=<-1 -2 -3>
      }
      localJointOffset=<-0.1 -0.2 0>
     }]
    relativeSpaceBone {boneName=Melee_TwoHanded_Handle}
   }
   n3 = Anim Dynamics - Bone: dyn_cube_l_out_01(ComponentPose=@n2, Alpha=0.4, ExternalForce="0,0,0") {
    angularDampingOverride=0.9
    angularSpringConstant=20000
    bAngularSpring=true
    bGravityOverrideInSimSpace=true
    bLinearSpring=true
    bOverrideAngularDamping=true
    bOverrideLinearDamping=true
    gravityScale=0
    linearDampingOverride=0.995
    linearSpringConstant=150
    simulationSpace=RootRelative
    alphaBoolBlend {blendOption=Linear}
    boundBone {boneName=dyn_cube_l_out_01}
    componentAppliedLinearAccClamp=<6000 6000 6000>
    componentLinearAccScale=<0.1 0.1 0.1>
    componentLinearVelScale=<0.9 0.9 0.9>
    gravityOverride {z=-980}
    physicsBodyDefinitions [{
      sphereCollisionRadius=0
      boundBone {boneName=dyn_cube_l_out_01}
      boxExtents=<15 15 15>
      constraintSetup {
       angularTargetAxis=AxisZ
       twistAxis=AxisZ
       angularLimitsMax=<0 0 0>
       angularLimitsMin=<0 0 0>
       angularTarget=<0 0 0>
       linearAxesMax=<1 3 2>
       linearAxesMin=<-1 -2 -3>
      }
      localJointOffset=<-0.1 -0.2 0>
     }]
    relativeSpaceBone {boneName=b_AR_Root}
   }
   n4 = Anim Dynamics - Bone: dyn_cube_l_out_02(ComponentPose=@n3, Alpha=0.45, ExternalForce="0,0,0") {
    angularDampingOverride=0.9
    angularSpringConstant=20000
    bAngularSpring=true
    bGravityOverrideInSimSpace=true
    bLinearSpring=true
    bOverrideAngularDamping=true
    bOverrideLinearDamping=true
    gravityScale=0
    linearDampingOverride=0.995
    linearSpringConstant=150
    simulationSpace=BoneRelative
    alphaBoolBlend {blendOption=Linear}
    boundBone {boneName=dyn_cube_l_out_02}
    componentAppliedLinearAccClamp=<6000 6000 6000>
    componentLinearAccScale=<0.1 0.1 0.1>
    componentLinearVelScale=<0.9 0.9 0.9>
    gravityOverride {z=-980}
    physicsBodyDefinitions [{
      sphereCollisionRadius=0
      boundBone {boneName=dyn_cube_l_out_02}
      boxExtents=<15 15 15>
      constraintSetup {
       angularTargetAxis=AxisZ
       twistAxis=AxisZ
       angularLimitsMax=<0 0 0>
       angularLimitsMin=<0 0 0>
       angularTarget=<0 0 0>
       linearAxesMax=<1 3 2>
       linearAxesMin=<-1 -3 -2>
      }
      localJointOffset=<-0.1 -0.2 0>
     }]
    relativeSpaceBone {boneName=dyn_master}
   }
   n5 = Anim Dynamics - Bone: dyn_cube_l_out_03(ComponentPose=@n4, Alpha=0.55, ExternalForce="0,0,0") {
    angularDampingOverride=0.9
    angularSpringConstant=20000
    bAngularSpring=true
    bGravityOverrideInSimSpace=true
    bLinearSpring=true
    bOverrideAngularDamping=true
    bOverrideLinearDamping=true
    gravityScale=0
    linearDampingOverride=0.995
    linearSpringConstant=150
    simulationSpace=BoneRelative
    alphaBoolBlend {blendOption=Linear}
    boundBone {boneName=dyn_cube_l_out_03}
    componentAppliedLinearAccClamp=<6000 6000 6000>
    componentLinearAccScale=<0.1 0.1 0.1>
    componentLinearVelScale=<0.9 0.9 0.9>
    gravityOverride {z=-980}
    physicsBodyDefinitions [{
      sphereCollisionRadius=0
      boundBone {boneName=dyn_cube_l_out_03}
      boxExtents=<15 15 15>
      constraintSetup {angularLimitsMax=<0 0 0>, angularLimitsMin=<0 0 0>, angularTarget=<0 0 0>, linearAxesMax=<1 2 3>, linearAxesMin=<-1 -2 -3>}
      localJointOffset=<-0.1 -0.2 0>
     }]
    relativeSpaceBone {boneName=dyn_master}
   }
   n6 = Anim Dynamics - Bone: dyn_cube_l_out_04(ComponentPose=@n5, Alpha=0.6, ExternalForce="0,0,0") {
    angularDampingOverride=0.9
    angularSpringConstant=20000
    bAngularSpring=true
    bGravityOverrideInSimSpace=true
    bLinearSpring=true
    bOverrideAngularDamping=true
    bOverrideLinearDamping=true
    gravityScale=0
    linearDampingOverride=0.995
    linearSpringConstant=150
    simulationSpace=RootRelative
    alphaBoolBlend {blendOption=Linear}
    boundBone {boneName=dyn_cube_l_out_04}
    componentAppliedLinearAccClamp=<6000 6000 6000>
    componentLinearAccScale=<0.1 0.1 0.1>
    componentLinearVelScale=<0.9 0.9 0.9>
    gravityOverride {z=-980}
    physicsBodyDefinitions [{
      sphereCollisionRadius=0
      boundBone {boneName=dyn_cube_l_out_04}
      boxExtents=<15 15 15>
      constraintSetup {
       angularTargetAxis=AxisZ
       twistAxis=AxisZ
       angularLimitsMax=<0 0 0>
       angularLimitsMin=<0 0 0>
       angularTarget=<0 0 0>
       linearAxesMax=<1 3 2>
       linearAxesMin=<-1 -3 -2>
      }
      localJointOffset=<-0.1 -0.2 0>
     }]
    relativeSpaceBone {boneName=b_AR_Root}
   }
   n7 = Anim Dynamics - Bone: dyn_cube_l_tip_01(ComponentPose=@n6, Alpha=0.65, ExternalForce="0,0,0") {
    angularDampingOverride=0.9
    angularSpringConstant=20000
    bAngularSpring=true
    bGravityOverrideInSimSpace=true
    bLinearSpring=true
    bOverrideAngularDamping=true
    bOverrideLinearDamping=true
    gravityScale=0
    linearDampingOverride=0.995
    linearSpringConstant=150
    simulationSpace=BoneRelative
    alphaBoolBlend {blendOption=Linear}
    boundBone {boneName=dyn_cube_l_tip_01}
    componentAppliedLinearAccClamp=<6000 6000 6000>
    componentLinearAccScale=<0.1 0.1 0.1>
    componentLinearVelScale=<0.9 0.9 0.9>
    gravityOverride {z=-980}
    physicsBodyDefinitions [{
      sphereCollisionRadius=0
      boundBone {boneName=dyn_cube_l_tip_01}
      boxExtents=<20 20 20>
      constraintSetup {
       angularTargetAxis=AxisZ
       twistAxis=AxisZ
       angularLimitsMax=<0 0 0>
       angularLimitsMin=<0 0 0>
       angularTarget=<0 0 0>
       linearAxesMax=<1 3 2>
       linearAxesMin=<-1 -2 -3>
      }
      localJointOffset=<-0.1 -0.2 0>
     }]
    relativeSpaceBone {boneName=dyn_master}
   }
   n8 = Anim Dynamics - Bone: dyn_cube_l_tip_02(ComponentPose=@n7, Alpha=0.7, ExternalForce="0,0,0") {
    angularDampingOverride=0.9
    angularSpringConstant=20000
    bAngularSpring=true
    bGravityOverrideInSimSpace=true
    bLinearSpring=true
    bOverrideAngularDamping=true
    bOverrideLinearDamping=true
    gravityScale=0
    linearDampingOverride=0.995
    linearSpringConstant=150
    simulationSpace=BoneRelative
    alphaBoolBlend {blendOption=Linear}
    boundBone {boneName=dyn_cube_l_tip_02}
    componentAppliedLinearAccClamp=<6000 6000 6000>
    componentLinearAccScale=<0.1 0.1 0.1>
    componentLinearVelScale=<0.9 0.9 0.9>
    gravityOverride {z=-980}
    physicsBodyDefinitions [{
      sphereCollisionRadius=0
      boundBone {boneName=dyn_cube_l_tip_02}
      boxExtents=<20 20 20>
      constraintSetup {angularLimitsMax=<0 0 0>, angularLimitsMin=<0 0 0>, angularTarget=<0 0 0>, linearAxesMax=<1 2 3>, linearAxesMin=<-1 -2 -3>}
      localJointOffset=<-0.1 -0.2 0>
     }]
    relativeSpaceBone {boneName=Melee_TwoHanded_Handle}
   }
   n9 = Anim Dynamics - Bone: dyn_cube_l_tip_03(ComponentPose=@n8, Alpha=0.75, ExternalForce="0,0,0") {
    angularDampingOverride=0.9
    angularSpringConstant=20000
    bAngularSpring=true
    bGravityOverrideInSimSpace=true
    bLinearSpring=true
    bOverrideAngularDamping=true
    bOverrideLinearDamping=true
    gravityScale=0
    linearDampingOverride=0.995
    linearSpringConstant=150
    simulationSpace=RootRelative
    alphaBoolBlend {blendOption=Linear}
    boundBone {boneName=dyn_cube_l_tip_03}
    componentAppliedLinearAccClamp=<6000 6000 6000>
    componentLinearAccScale=<0.1 0.1 0.1>
    componentLinearVelScale=<0.9 0.9 0.9>
    gravityOverride {z=-980}
    physicsBodyDefinitions [{
      sphereCollisionRadius=0
      boundBone {boneName=dyn_cube_l_tip_03}
      boxExtents=<20 20 20>
      constraintSetup {
       angularTargetAxis=AxisZ
       twistAxis=AxisZ
       angularLimitsMax=<0 0 0>
       angularLimitsMin=<0 0 0>
       angularTarget=<0 0 0>
       linearAxesMax=<1 3 2>
       linearAxesMin=<-1 -3 -2>
      }
      localJointOffset=<-0.1 -0.2 0>
     }]
    relativeSpaceBone {boneName=b_AR_Root}
   }
   Anim Dynamics - Bone: dyn_cube_l_tip_04(ComponentPose=@n9, Alpha=0.85, ExternalForce="0,0,0") {
    angularDampingOverride=0.9
    angularSpringConstant=20000
    bAngularSpring=true
    bGravityOverrideInSimSpace=true
    bLinearSpring=true
    bOverrideAngularDamping=true
    bOverrideLinearDamping=true
    gravityScale=0
    linearDampingOverride=0.995
    linearSpringConstant=150
    simulationSpace=RootRelative
    alphaBoolBlend {blendOption=Linear}
    boundBone {boneName=dyn_cube_l_tip_04}
    componentAppliedLinearAccClamp=<6000 6000 6000>
    componentLinearAccScale=<0.1 0.1 0.1>
    componentLinearVelScale=<0.9 0.9 0.9>
    gravityOverride {z=-980}
    physicsBodyDefinitions [{
      sphereCollisionRadius=0
      boundBone {boneName=dyn_cube_l_tip_04}
      boxExtents=<20 20 20>
      constraintSetup {angularLimitsMax=<0 0 0>, angularLimitsMin=<0 0 0>, angularTarget=<0 0 0>, linearAxesMax=<1 2 3>, linearAxesMin=<-1 -2 -3>}
      localJointOffset=<-0.1 -0.2 0>
     }]
    relativeSpaceBone {boneName=b_AR_Root}
   }
   n10 = Anim Dynamics - Bone: dyn_cube_l_in(ComponentPose=Local To Component(LocalPose=Use cached pose 'CubesFloat'()), Alpha=0.2, ExternalForce="0,0,0") {
    angularDampingOverride=0.9
    angularSpringConstant=200000
    bAngularSpring=true
    bGravityOverrideInSimSpace=true
    bLinearSpring=true
    bOverrideAngularDamping=true
    bOverrideLinearDamping=true
    gravityScale=0
    linearDampingOverride=0.995
    linearSpringConstant=50
    simulationSpace=BoneRelative
    alphaBoolBlend {blendOption=Linear}
    boundBone {boneName=dyn_cube_l_in}
    componentAppliedLinearAccClamp=<6000 6000 6000>
    componentLinearAccScale=<0.3 0.3 0.1>
    componentLinearVelScale=<0.7 0.7 0.8>
    gravityOverride {z=-980}
    physicsBodyDefinitions [{
      sphereCollisionRadius=0
      boundBone {boneName=dyn_cube_l_in}
      boxExtents=<5 5 5>
      constraintSetup {
       angularTargetAxis=AxisZ
       twistAxis=AxisZ
       angularLimitsMax=<0 0 0>
       angularLimitsMin=<0 0 0>
       angularTarget=<0 0 0>
       linearAxesMax=<18 18 18>
       linearAxesMin=<-1 -0.5 -2>
      }
      localJointOffset=<-0.1 -0.2 0>
     }]
    relativeSpaceBone {boneName=dyn_master}
   }
   n11 = Anim Dynamics - Bone: dyn_cube_l_mid(ComponentPose=@n10, Alpha=0.35, ExternalForce="0,0,0") {
    angularDampingOverride=0.9
    angularSpringConstant=200000
    bAngularSpring=true
    bGravityOverrideInSimSpace=true
    bLinearSpring=true
    bOverrideAngularDamping=true
    bOverrideLinearDamping=true
    gravityScale=0
    linearDampingOverride=0.995
    linearSpringConstant=50
    simulationSpace=BoneRelative
    alphaBoolBlend {blendOption=Linear}
    boundBone {boneName=dyn_cube_l_mid}
    componentAppliedLinearAccClamp=<6000 6000 6000>
    componentLinearAccScale=<0.3 0.3 0.1>
    componentLinearVelScale=<0.7 0.7 0.8>
    gravityOverride {z=-980}
    physicsBodyDefinitions [{
      sphereCollisionRadius=0
      boundBone {boneName=dyn_cube_l_mid}
      boxExtents=<10 10 10>
      constraintSetup {
       angularTargetAxis=AxisZ
       twistAxis=AxisZ
       angularLimitsMax=<0 0 0>
       angularLimitsMin=<0 0 0>
       angularTarget=<0 0 0>
       linearAxesMax=<18 18 18>
       linearAxesMin=<-1 -2 -2>
      }
      localJointOffset=<-0.1 -0.2 0>
     }]
    relativeSpaceBone {boneName=dyn_master}
   }
   n12 = Anim Dynamics - Bone: dyn_cube_l_out_01(ComponentPose=@n11, Alpha=0.4, ExternalForce="0,0,0") {
    angularDampingOverride=0.9
    angularSpringConstant=200000
    bAngularSpring=true
    bGravityOverrideInSimSpace=true
    bLinearSpring=true
    bOverrideAngularDamping=true
    bOverrideLinearDamping=true
    gravityScale=0
    linearDampingOverride=0.995
    linearSpringConstant=50
    simulationSpace=BoneRelative
    alphaBoolBlend {blendOption=Linear}
    boundBone {boneName=dyn_cube_l_out_01}
    componentAppliedLinearAccClamp=<6000 6000 6000>
    componentLinearAccScale=<0.3 0.3 0.1>
    componentLinearVelScale=<0.7 0.7 0.8>
    gravityOverride {z=-980}
    physicsBodyDefinitions [{
      sphereCollisionRadius=0
      boundBone {boneName=dyn_cube_l_out_01}
      boxExtents=<15 15 15>
      constraintSetup {
       angularTargetAxis=AxisZ
       twistAxis=AxisZ
       angularLimitsMax=<0 0 0>
       angularLimitsMin=<0 0 0>
       angularTarget=<0 0 0>
       linearAxesMax=<18 18 18>
       linearAxesMin=<-0.1 -1.2 -2>
      }
      localJointOffset=<-0.1 -0.2 0>
     }]
    relativeSpaceBone {boneName=dyn_master}
   }
   n13 = Anim Dynamics - Bone: dyn_cube_l_out_02(ComponentPose=@n12, Alpha=0.2, ExternalForce="0,0,0") {
    angularDampingOverride=0.9
    angularSpringConstant=200000
    bAngularSpring=true
    bGravityOverrideInSimSpace=true
    bLinearSpring=true
    bOverrideAngularDamping=true
    bOverrideLinearDamping=true
    gravityScale=0
    linearDampingOverride=0.995
    linearSpringConstant=50
    simulationSpace=BoneRelative
    alphaBoolBlend {blendOption=Linear}
    boundBone {boneName=dyn_cube_l_out_02}
    componentAppliedLinearAccClamp=<6000 6000 6000>
    componentLinearAccScale=<0.3 0.3 0.1>
    componentLinearVelScale=<0.7 0.7 0.8>
    gravityOverride {z=-980}
    physicsBodyDefinitions [{
      sphereCollisionRadius=0
      boundBone {boneName=dyn_cube_l_out_02}
      boxExtents=<15 15 15>
      constraintSetup {
       angularTargetAxis=AxisZ
       twistAxis=AxisZ
       angularLimitsMax=<0 0 0>
       angularLimitsMin=<0 0 0>
       angularTarget=<0 0 0>
       linearAxesMax=<18 18 18>
       linearAxesMin=<-0.1 -1.5 -2>
      }
      localJointOffset=<-0.1 0 0>
     }]
    relativeSpaceBone {boneName=dyn_master}
   }
   n14 = Anim Dynamics - Bone: dyn_cube_l_out_03(ComponentPose=@n13, Alpha=0.5, ExternalForce="0,0,0") {
    angularDampingOverride=0.9
    angularSpringConstant=200000
    bAngularSpring=true
    bGravityOverrideInSimSpace=true
    bLinearSpring=true
    bOverrideAngularDamping=true
    bOverrideLinearDamping=true
    gravityScale=0
    linearDampingOverride=0.995
    linearSpringConstant=50
    simulationSpace=BoneRelative
    alphaBoolBlend {blendOption=Linear}
    boundBone {boneName=dyn_cube_l_out_03}
    componentAppliedLinearAccClamp=<6000 6000 6000>
    componentLinearAccScale=<0.3 0.3 0.1>
    componentLinearVelScale=<0.7 0.7 0.8>
    gravityOverride {z=-980}
    physicsBodyDefinitions [{
      sphereCollisionRadius=0
      boundBone {boneName=dyn_cube_l_out_03}
      boxExtents=<15 15 15>
      constraintSetup {
       angularTargetAxis=AxisZ
       twistAxis=AxisZ
       angularLimitsMax=<0 0 0>
       angularLimitsMin=<0 0 0>
       angularTarget=<0 0 0>
       linearAxesMax=<18 18 18>
       linearAxesMin=<-0.4 -2 -2>
      }
      localJointOffset=<-0.1 -0.2 0>
     }]
    relativeSpaceBone {boneName=dyn_master}
   }
   n15 = Anim Dynamics - Bone: dyn_cube_l_out_04(ComponentPose=@n14, Alpha=0.3, ExternalForce="0,0,0") {
    angularDampingOverride=0.9
    angularSpringConstant=200000
    bAngularSpring=true
    bGravityOverrideInSimSpace=true
    bLinearSpring=true
    bOverrideAngularDamping=true
    bOverrideLinearDamping=true
    gravityScale=0
    linearDampingOverride=0.995
    linearSpringConstant=50
    simulationSpace=BoneRelative
    alphaBoolBlend {blendOption=Linear}
    boundBone {boneName=dyn_cube_l_out_04}
    componentAppliedLinearAccClamp=<6000 6000 6000>
    componentLinearAccScale=<0.3 0.3 0.1>
    componentLinearVelScale=<0.7 0.7 0.8>
    gravityOverride {z=-980}
    physicsBodyDefinitions [{
      sphereCollisionRadius=0
      boundBone {boneName=dyn_cube_l_out_04}
      boxExtents=<15 15 15>
      constraintSetup {
       angularTargetAxis=AxisZ
       twistAxis=AxisZ
       angularLimitsMax=<0 0 0>
       angularLimitsMin=<0 0 0>
       angularTarget=<0 0 0>
       linearAxesMax=<18 18 18>
       linearAxesMin=<-0.35 -1.65 -2>
      }
      localJointOffset=<-0.1 -0.2 0>
     }]
    relativeSpaceBone {boneName=dyn_master}
   }
   n16 = Anim Dynamics - Bone: dyn_cube_l_tip_01(ComponentPose=@n15, Alpha=0.3, ExternalForce="0,0,0") {
    angularDampingOverride=0.9
    angularSpringConstant=200000
    bAngularSpring=true
    bGravityOverrideInSimSpace=true
    bLinearSpring=true
    bOverrideAngularDamping=true
    bOverrideLinearDamping=true
    gravityScale=0
    linearDampingOverride=0.995
    linearSpringConstant=50
    simulationSpace=BoneRelative
    alphaBoolBlend {blendOption=Linear}
    boundBone {boneName=dyn_cube_l_tip_01}
    componentAppliedLinearAccClamp=<6000 6000 6000>
    componentLinearAccScale=<0.3 0.3 0.1>
    componentLinearVelScale=<0.7 0.7 0.8>
    gravityOverride {z=-980}
    physicsBodyDefinitions [{
      sphereCollisionRadius=0
      boundBone {boneName=dyn_cube_l_tip_01}
      boxExtents=<20 20 20>
      constraintSetup {
       angularTargetAxis=AxisZ
       twistAxis=AxisZ
       angularLimitsMax=<0 0 0>
       angularLimitsMin=<0 0 0>
       angularTarget=<0 0 0>
       linearAxesMax=<18 18 18>
       linearAxesMin=<-0.1 -1.5 -2>
      }
      localJointOffset=<-0.1 -0.2 0>
     }]
    relativeSpaceBone {boneName=dyn_master}
   }
   n17 = Anim Dynamics - Bone: dyn_cube_l_tip_02(ComponentPose=@n16, Alpha=0.5, ExternalForce="0,0,0") {
    angularDampingOverride=0.9
    angularSpringConstant=200000
    bAngularSpring=true
    bGravityOverrideInSimSpace=true
    bLinearSpring=true
    bOverrideAngularDamping=true
    bOverrideLinearDamping=true
    gravityScale=0
    linearDampingOverride=0.995
    linearSpringConstant=50
    simulationSpace=BoneRelative
    alphaBoolBlend {blendOption=Linear}
    boundBone {boneName=dyn_cube_l_tip_02}
    componentAppliedLinearAccClamp=<6000 6000 6000>
    componentLinearAccScale=<0.3 0.3 0.1>
    componentLinearVelScale=<0.7 0.7 0.8>
    gravityOverride {z=-980}
    physicsBodyDefinitions [{
      sphereCollisionRadius=0
      boundBone {boneName=dyn_cube_l_tip_02}
      boxExtents=<20 20 20>
      constraintSetup {
       angularTargetAxis=AxisZ
       twistAxis=AxisZ
       angularLimitsMax=<0 0 0>
       angularLimitsMin=<0 0 0>
       angularTarget=<0 0 0>
       linearAxesMax=<18 18 18>
       linearAxesMin=<-0.5 -1 -2>
      }
      localJointOffset=<-0.1 -0.2 0>
     }]
    relativeSpaceBone {boneName=dyn_master}
   }
   n18 = Anim Dynamics - Bone: dyn_cube_l_tip_03(ComponentPose=@n17, Alpha=0.25, ExternalForce="0,0,0") {
    angularDampingOverride=0.9
    angularSpringConstant=200000
    bAngularSpring=true
    bGravityOverrideInSimSpace=true
    bLinearSpring=true
    bOverrideAngularDamping=true
    bOverrideLinearDamping=true
    gravityScale=0
    linearDampingOverride=0.995
    linearSpringConstant=50
    simulationSpace=BoneRelative
    alphaBoolBlend {blendOption=Linear}
    boundBone {boneName=dyn_cube_l_tip_03}
    componentAppliedLinearAccClamp=<6000 6000 6000>
    componentLinearAccScale=<0.3 0.3 0.1>
    componentLinearVelScale=<0.7 0.7 0.8>
    gravityOverride {z=-980}
    physicsBodyDefinitions [{
      sphereCollisionRadius=0
      boundBone {boneName=dyn_cube_l_tip_03}
      boxExtents=<20 20 20>
      constraintSetup {
       angularTargetAxis=AxisZ
       twistAxis=AxisZ
       angularLimitsMax=<0 0 0>
       angularLimitsMin=<0 0 0>
       angularTarget=<0 0 0>
       linearAxesMax=<18 18 18>
       linearAxesMin=<-0.1 -1.5 -2>
      }
      localJointOffset=<-0.1 -0.2 0>
     }]
    relativeSpaceBone {boneName=dyn_master}
   }
   n19 = Anim Dynamics - Bone: dyn_cube_l_tip_04(ComponentPose=@n18, Alpha=0.55, ExternalForce="0,0,0") {
    angularDampingOverride=0.9
    angularSpringConstant=200000
    bAngularSpring=true
    bGravityOverrideInSimSpace=true
    bLinearSpring=true
    bOverrideAngularDamping=true
    bOverrideLinearDamping=true
    gravityScale=0
    linearDampingOverride=0.995
    linearSpringConstant=50
    simulationSpace=BoneRelative
    alphaBoolBlend {blendOption=Linear}
    boundBone {boneName=dyn_cube_l_tip_04}
    componentAppliedLinearAccClamp=<6000 6000 6000>
    componentLinearAccScale=<0.3 0.3 0.1>
    componentLinearVelScale=<0.7 0.7 0.8>
    gravityOverride {z=-980}
    physicsBodyDefinitions [{
      sphereCollisionRadius=0
      boundBone {boneName=dyn_cube_l_tip_04}
      boxExtents=<20 20 20>
      constraintSetup {
       angularTargetAxis=AxisZ
       twistAxis=AxisZ
       angularLimitsMax=<0 0 0>
       angularLimitsMin=<0 0 0>
       angularTarget=<0 0 0>
       linearAxesMax=<18 18 18>
       linearAxesMin=<-1 -1.5 -2>
      }
      localJointOffset=<-0.1 -0.2 0>
     }]
    relativeSpaceBone {boneName=dyn_master}
   }
   n20 = Anim Dynamics - Bone: dyn_cube_r_in_01(ComponentPose=@n19, Alpha=0.2, ExternalForce="0,0,0") {
    angularDampingOverride=0.9
    angularSpringConstant=200000
    bAngularSpring=true
    bGravityOverrideInSimSpace=true
    bLinearSpring=true
    bOverrideAngularDamping=true
    bOverrideLinearDamping=true
    gravityScale=0
    linearDampingOverride=0.995
    linearSpringConstant=50
    simulationSpace=BoneRelative
    alphaBoolBlend {blendOption=Linear}
    boundBone {boneName=dyn_cube_r_in_01}
    componentAppliedLinearAccClamp=<6000 6000 6000>
    componentLinearAccScale=<0.3 0.3 0.1>
    componentLinearVelScale=<0.7 0.7 0.8>
    gravityOverride {z=-980}
    physicsBodyDefinitions [{
      sphereCollisionRadius=0
      boundBone {boneName=dyn_cube_r_in_01}
      boxExtents=<5 5 5>
      constraintSetup {
       angularTargetAxis=AxisZ
       twistAxis=AxisZ
       angularLimitsMax=<0 0 0>
       angularLimitsMin=<0 0 0>
       angularTarget=<0 0 0>
       linearAxesMax=<18 18 18>
       linearAxesMin=<-0.2 -1.35 -2>
      }
      localJointOffset=<-0.1 -0.2 0>
     }]
    relativeSpaceBone {boneName=dyn_master}
   }
   n21 = Anim Dynamics - Bone: dyn_cube_r_in_02(ComponentPose=@n20, Alpha=0.4, ExternalForce="0,0,0") {
    angularDampingOverride=0.9
    angularSpringConstant=200000
    bAngularSpring=true
    bGravityOverrideInSimSpace=true
    bLinearSpring=true
    bOverrideAngularDamping=true
    bOverrideLinearDamping=true
    gravityScale=0
    linearDampingOverride=0.995
    linearSpringConstant=50
    simulationSpace=BoneRelative
    alphaBoolBlend {blendOption=Linear}
    boundBone {boneName=dyn_cube_r_in_02}
    componentAppliedLinearAccClamp=<6000 6000 6000>
    componentLinearAccScale=<0.3 0.3 0.1>
    componentLinearVelScale=<0.7 0.7 0.8>
    gravityOverride {z=-980}
    physicsBodyDefinitions [{
      sphereCollisionRadius=0
      boundBone {boneName=dyn_cube_r_in_02}
      boxExtents=<5 5 5>
      constraintSetup {
       angularTargetAxis=AxisZ
       twistAxis=AxisZ
       angularLimitsMax=<0 0 0>
       angularLimitsMin=<0 0 0>
       angularTarget=<0 0 0>
       linearAxesMax=<18 18 18>
       linearAxesMin=<-0.1 -0.1 -2>
      }
      localJointOffset=<-0.1 -0.2 0>
     }]
    relativeSpaceBone {boneName=dyn_master}
   }
   n22 = Anim Dynamics - Bone: dyn_cube_r_mid(ComponentPose=@n21, Alpha=0.5, ExternalForce="0,0,0") {
    angularDampingOverride=0.9
    angularSpringConstant=200000
    bAngularSpring=true
    bGravityOverrideInSimSpace=true
    bLinearSpring=true
    bOverrideAngularDamping=true
    bOverrideLinearDamping=true
    gravityScale=0
    linearDampingOverride=0.995
    linearSpringConstant=50
    simulationSpace=BoneRelative
    alphaBoolBlend {blendOption=Linear}
    boundBone {boneName=dyn_cube_r_mid}
    componentAppliedLinearAccClamp=<6000 6000 6000>
    componentLinearAccScale=<0.3 0.3 0.1>
    componentLinearVelScale=<0.7 0.7 0.8>
    gravityOverride {z=-980}
    physicsBodyDefinitions [{
      sphereCollisionRadius=0
      boundBone {boneName=dyn_cube_r_mid}
      boxExtents=<10 10 10>
      constraintSetup {
       angularTargetAxis=AxisZ
       twistAxis=AxisZ
       angularLimitsMax=<0 0 0>
       angularLimitsMin=<0 0 0>
       angularTarget=<0 0 0>
       linearAxesMax=<18 18 18>
       linearAxesMin=<-0.5 -1.35 -2>
      }
      localJointOffset=<-0.1 -0.2 0>
     }]
    relativeSpaceBone {boneName=dyn_master}
   }
   n23 = Anim Dynamics - Bone: dyn_cube_r_tip_01(ComponentPose=@n22, Alpha=0.35, ExternalForce="0,0,0") {
    angularDampingOverride=0.9
    angularSpringConstant=200000
    bAngularSpring=true
    bGravityOverrideInSimSpace=true
    bLinearSpring=true
    bOverrideAngularDamping=true
    bOverrideLinearDamping=true
    gravityScale=0
    linearDampingOverride=0.995
    linearSpringConstant=50
    simulationSpace=BoneRelative
    alphaBoolBlend {blendOption=Linear}
    boundBone {boneName=dyn_cube_r_tip_01}
    componentAppliedLinearAccClamp=<6000 6000 6000>
    componentLinearAccScale=<0.3 0.3 0.1>
    componentLinearVelScale=<0.7 0.7 0.8>
    gravityOverride {z=-980}
    physicsBodyDefinitions [{
      sphereCollisionRadius=0
      boundBone {boneName=dyn_cube_r_tip_01}
      boxExtents=<20 20 20>
      constraintSetup {
       angularTargetAxis=AxisZ
       twistAxis=AxisZ
       angularLimitsMax=<0 0 0>
       angularLimitsMin=<0 0 0>
       angularTarget=<0 0 0>
       linearAxesMax=<18 18 18>
       linearAxesMin=<-0.5 -1 -2>
      }
      localJointOffset=<-0.1 -0.2 0>
     }]
    relativeSpaceBone {boneName=dyn_master}
   }
   n24 = Anim Dynamics - Bone: dyn_cube_r_tip_02(ComponentPose=@n23, Alpha=0.45, ExternalForce="0,0,0") {
    angularDampingOverride=0.9
    angularSpringConstant=200000
    bAngularSpring=true
    bGravityOverrideInSimSpace=true
    bLinearSpring=true
    bOverrideAngularDamping=true
    bOverrideLinearDamping=true
    gravityScale=0
    linearDampingOverride=0.995
    linearSpringConstant=50
    simulationSpace=BoneRelative
    alphaBoolBlend {blendOption=Linear}
    boundBone {boneName=dyn_cube_r_tip_02}
    componentAppliedLinearAccClamp=<6000 6000 6000>
    componentLinearAccScale=<0.3 0.3 0.1>
    componentLinearVelScale=<0.7 0.7 0.8>
    gravityOverride {z=-980}
    physicsBodyDefinitions [{
      sphereCollisionRadius=0
      boundBone {boneName=dyn_cube_r_tip_02}
      boxExtents=<20 20 20>
      constraintSetup {
       angularTargetAxis=AxisZ
       twistAxis=AxisZ
       angularLimitsMax=<0 0 0>
       angularLimitsMin=<0 0 0>
       angularTarget=<0 0 0>
       linearAxesMax=<18 18 18>
       linearAxesMin=<-0.2 -1 -2>
      }
      localJointOffset=<-0.1 -0.2 0>
     }]
    relativeSpaceBone {boneName=dyn_master}
   }
   n25 = Anim Dynamics - Bone: dyn_cube_r_tip_03(ComponentPose=@n24, Alpha=0.55, ExternalForce="0,0,0") {
    angularDampingOverride=0.9
    angularSpringConstant=200000
    bAngularSpring=true
    bGravityOverrideInSimSpace=true
    bLinearSpring=true
    bOverrideAngularDamping=true
    bOverrideLinearDamping=true
    gravityScale=0
    linearDampingOverride=0.995
    linearSpringConstant=50
    simulationSpace=BoneRelative
    alphaBoolBlend {blendOption=Linear}
    boundBone {boneName=dyn_cube_r_tip_03}
    componentAppliedLinearAccClamp=<6000 6000 6000>
    componentLinearAccScale=<0.3 0.3 0.1>
    componentLinearVelScale=<0.7 0.7 0.8>
    gravityOverride {z=-980}
    physicsBodyDefinitions [{
      sphereCollisionRadius=0
      boundBone {boneName=dyn_cube_r_tip_03}
      boxExtents=<20 20 20>
      constraintSetup {
       angularTargetAxis=AxisZ
       twistAxis=AxisZ
       angularLimitsMax=<0 0 0>
       angularLimitsMin=<0 0 0>
       angularTarget=<0 0 0>
       linearAxesMax=<18 18 18>
       linearAxesMin=<-1 -1 -2>
      }
      localJointOffset=<-0.1 -0.2 0>
     }]
    relativeSpaceBone {boneName=dyn_master}
   }
   n26 = Anim Dynamics - Bone: dyn_master(ComponentPose=Local To Component(LocalPose=Use cached pose 'CubesFloat'()), Alpha=0.9, ExternalForce="0,0,0") {
    angularDampingOverride=0.9
    angularSpringConstant=175
    bAngularSpring=true
    bGravityOverrideInSimSpace=true
    bLinearSpring=true
    bOverrideAngularDamping=true
    bOverrideLinearDamping=true
    gravityScale=0
    linearDampingOverride=0.995
    linearSpringConstant=155
    simulationSpace=BoneRelative
    alphaBoolBlend {blendOption=Linear}
    boundBone {boneName=dyn_master}
    componentAppliedLinearAccClamp=<6000 6000 6000>
    componentLinearVelScale=<0.9 0.9 0.9>
    physicsBodyDefinitions [{
      sphereCollisionRadius=0
      boundBone {boneName=dyn_master}
      boxExtents=<15 15 15>
      constraintSetup {angularLimitsMax=<0 0 0>, angularLimitsMin=<0 0 0>, angularTarget=<0 0 0>, linearAxesMax=<7 8 8>, linearAxesMin=<-8 -2 -5>}
      localJointOffset=<-3 -3 -3>
     }]
    relativeSpaceBone {boneName=dyn_cube_l_in}
   }
   n27 = Anim Dynamics - Bone: dyn_cube_l_mid(ComponentPose=@n26, Alpha=0.9, ExternalForce="0,0,0") {
    angularDampingOverride=0.9
    angularSpringConstant=195
    bAngularSpring=true
    bGravityOverrideInSimSpace=true
    bLinearSpring=true
    bOverrideAngularDamping=true
    bOverrideLinearDamping=true
    gravityScale=0
    linearDampingOverride=0.995
    linearSpringConstant=155
    simulationSpace=BoneRelative
    alphaBoolBlend {blendOption=Linear}
    boundBone {boneName=dyn_cube_l_mid}
    componentAppliedLinearAccClamp=<6000 6000 6000>
    componentLinearVelScale=<0.9 0.9 0.9>
    physicsBodyDefinitions [{
      sphereCollisionRadius=0
      boundBone {boneName=dyn_cube_l_mid}
      boxExtents=<15 15 15>
      constraintSetup {angularLimitsMax=<0 0 0>, angularLimitsMin=<0 0 0>, angularTarget=<0 0 0>, linearAxesMax=<8 3 8>, linearAxesMin=<-5 -8 -4>}
      localJointOffset=<3 3 3>
     }]
    relativeSpaceBone {boneName=Melee_TwoHanded_Handle}
   }
   n28 = Anim Dynamics - Bone: dyn_cube_l_out_01(ComponentPose=@n27, Alpha=0.9, ExternalForce="0,0,0") {
    angularDampingOverride=0.9
    angularSpringConstant=195
    bAngularSpring=true
    bGravityOverrideInSimSpace=true
    bLinearSpring=true
    bOverrideAngularDamping=true
    bOverrideLinearDamping=true
    gravityScale=0
    linearDampingOverride=0.995
    linearSpringConstant=155
    simulationSpace=BoneRelative
    alphaBoolBlend {blendOption=Linear}
    boundBone {boneName=dyn_cube_l_out_01}
    componentAppliedLinearAccClamp=<6000 6000 6000>
    componentLinearVelScale=<0.9 0.9 0.9>
    physicsBodyDefinitions [{
      sphereCollisionRadius=0
      boundBone {boneName=dyn_cube_l_out_01}
      boxExtents=<50 50 50>
      constraintSetup {angularLimitsMax=<0 0 0>, angularLimitsMin=<0 0 0>, angularTarget=<0 0 0>, linearAxesMax=<8 8 6>, linearAxesMin=<-8 -2 -3>}
      localJointOffset=<8 8 8>
     }]
    relativeSpaceBone {boneName=Melee_TwoHanded_Handle}
   }
   n29 = Anim Dynamics - Bone: dyn_cube_l_out_02(ComponentPose=@n28, Alpha=0.9, ExternalForce="0,0,0") {
    angularDampingOverride=0.9
    angularSpringConstant=175
    bAngularSpring=true
    bGravityOverrideInSimSpace=true
    bLinearSpring=true
    bOverrideAngularDamping=true
    bOverrideLinearDamping=true
    gravityScale=0
    linearDampingOverride=0.995
    linearSpringConstant=155
    simulationSpace=BoneRelative
    alphaBoolBlend {blendOption=Linear}
    boundBone {boneName=dyn_cube_l_out_02}
    componentAppliedLinearAccClamp=<6000 6000 6000>
    componentLinearVelScale=<0.9 0.9 0.9>
    physicsBodyDefinitions [{
      sphereCollisionRadius=0
      boundBone {boneName=dyn_cube_l_out_02}
      boxExtents=<50 50 50>
      constraintSetup {angularLimitsMax=<0 0 0>, angularLimitsMin=<0 0 0>, angularTarget=<0 0 0>, linearAxesMax=<2 3 8>, linearAxesMin=<-8 -8 -2>}
      localJointOffset=<-2 -2 -2>
     }]
    relativeSpaceBone {boneName=Melee_TwoHanded_Handle}
   }
   n30 = Anim Dynamics - Bone: dyn_cube_l_out_03(ComponentPose=@n29, Alpha=0.9, ExternalForce="0,0,0") {
    angularDampingOverride=0.9
    angularSpringConstant=175
    bAngularSpring=true
    bGravityOverrideInSimSpace=true
    bLinearSpring=true
    bOverrideAngularDamping=true
    bOverrideLinearDamping=true
    gravityScale=0
    linearDampingOverride=0.995
    linearSpringConstant=155
    simulationSpace=BoneRelative
    alphaBoolBlend {blendOption=Linear}
    boundBone {boneName=dyn_cube_l_out_03}
    componentAppliedLinearAccClamp=<6000 6000 6000>
    componentLinearVelScale=<0.9 0.9 0.9>
    physicsBodyDefinitions [{
      sphereCollisionRadius=0
      boundBone {boneName=dyn_cube_l_out_03}
      boxExtents=<50 50 50>
      constraintSetup {angularLimitsMax=<0 0 0>, angularLimitsMin=<0 0 0>, angularTarget=<0 0 0>, linearAxesMax=<2 3 8>, linearAxesMin=<-8 -8 -2>}
      localJointOffset=<-3 -3 -3>
     }]
    relativeSpaceBone {boneName=Melee_TwoHanded_Handle}
   }
   n31 = Anim Dynamics - Bone: dyn_cube_l_out_04(ComponentPose=@n30, Alpha=0.9, ExternalForce="0,0,0") {
    angularDampingOverride=0.9
    angularSpringConstant=175
    bAngularSpring=true
    bGravityOverrideInSimSpace=true
    bLinearSpring=true
    bOverrideAngularDamping=true
    bOverrideLinearDamping=true
    gravityScale=0
    linearDampingOverride=0.995
    linearSpringConstant=155
    simulationSpace=BoneRelative
    alphaBoolBlend {blendOption=Linear}
    boundBone {boneName=dyn_cube_l_out_04}
    componentAppliedLinearAccClamp=<6000 6000 6000>
    componentLinearVelScale=<0.9 0.9 0.9>
    physicsBodyDefinitions [{
      sphereCollisionRadius=0
      boundBone {boneName=dyn_cube_l_out_04}
      boxExtents=<50 50 50>
      constraintSetup {angularLimitsMax=<0 0 0>, angularLimitsMin=<0 0 0>, angularTarget=<0 0 0>, linearAxesMax=<2 3 8>, linearAxesMin=<-8 -8 -2>}
      localJointOffset=<-10 -10 -10>
     }]
    relativeSpaceBone {boneName=Melee_TwoHanded_Handle}
   }
   n32 = Anim Dynamics - Bone: dyn_cube_l_tip_01(ComponentPose=@n31, Alpha=0.9, ExternalForce="0,0,0") {
    angularDampingOverride=0.9
    angularSpringConstant=175
    bAngularSpring=true
    bGravityOverrideInSimSpace=true
    bLinearSpring=true
    bOverrideAngularDamping=true
    bOverrideLinearDamping=true
    gravityScale=0
    linearDampingOverride=0.995
    linearSpringConstant=155
    simulationSpace=BoneRelative
    alphaBoolBlend {blendOption=Linear}
    boundBone {boneName=dyn_cube_l_tip_01}
    componentAppliedLinearAccClamp=<6000 6000 6000>
    componentLinearVelScale=<0.9 0.9 0.9>
    physicsBodyDefinitions [{
      sphereCollisionRadius=0
      boundBone {boneName=dyn_cube_l_tip_01}
      boxExtents=<50 50 50>
      constraintSetup {angularLimitsMax=<0 0 0>, angularLimitsMin=<0 0 0>, angularTarget=<0 0 0>, linearAxesMax=<4 8 8>, linearAxesMin=<-8 -2 -8>}
      localJointOffset=<-3 0 -3>
     }]
    relativeSpaceBone {boneName=Melee_TwoHanded_Handle}
   }
   n33 = Anim Dynamics - Bone: dyn_cube_l_tip_02(ComponentPose=@n32, Alpha=0.5, ExternalForce="0,0,0") {
    angularDampingOverride=0.9
    angularSpringConstant=150
    bAngularSpring=true
    bGravityOverrideInSimSpace=true
    bLinearSpring=true
    bOverrideAngularDamping=true
    bOverrideLinearDamping=true
    gravityScale=0
    linearDampingOverride=0.995
    linearSpringConstant=50
    simulationSpace=BoneRelative
    alphaBoolBlend {blendOption=Linear}
    boundBone {boneName=dyn_cube_l_tip_02}
    componentAppliedLinearAccClamp=<6000 6000 6000>
    componentLinearAccScale=<0.1 0.1 0.1>
    componentLinearVelScale=<0.9 0.9 0.9>
    physicsBodyDefinitions [{
      sphereCollisionRadius=0
      boundBone {boneName=dyn_cube_l_tip_02}
      boxExtents=<50 50 50>
      constraintSetup {angularLimitsMax=<5 5 5>, angularLimitsMin=<-5 -5 -5>, angularTarget=<0 0 0>, linearAxesMax=<2 8 2>, linearAxesMin=<-8 -8 -8>}
      localJointOffset=<-2 -2 -2>
     }]
    relativeSpaceBone {boneName=Melee_TwoHanded_Handle}
   }
   n34 = Anim Dynamics - Bone: dyn_cube_l_tip_03(ComponentPose=@n33, Alpha=1, ExternalForce="0,0,0") {
    angularDampingOverride=0.9
    angularSpringConstant=175
    bAngularSpring=true
    bGravityOverrideInSimSpace=true
    bLinearSpring=true
    bOverrideAngularDamping=true
    bOverrideLinearDamping=true
    gravityScale=0
    linearDampingOverride=0.995
    linearSpringConstant=155
    simulationSpace=BoneRelative
    alphaBoolBlend {blendOption=Linear}
    boundBone {boneName=dyn_cube_l_tip_03}
    componentAppliedLinearAccClamp=<6000 6000 6000>
    componentLinearVelScale=<0.9 0.9 0.9>
    physicsBodyDefinitions [{
      sphereCollisionRadius=0
      boundBone {boneName=dyn_cube_l_tip_03}
      boxExtents=<50 50 50>
      constraintSetup {angularLimitsMax=<0 0 0>, angularLimitsMin=<0 0 0>, angularTarget=<0 0 0>, linearAxesMax=<2 8 8>, linearAxesMin=<-8 -3 -8>}
      localJointOffset=<0 0 0>
     }]
    relativeSpaceBone {boneName=Melee_TwoHanded_Handle}
   }
   n35 = Anim Dynamics - Bone: dyn_cube_l_tip_04(ComponentPose=@n34, Alpha=1, ExternalForce="0,0,0") {
    angularDampingOverride=0.9
    angularSpringConstant=175
    bAngularSpring=true
    bGravityOverrideInSimSpace=true
    bLinearSpring=true
    bOverrideAngularDamping=true
    bOverrideLinearDamping=true
    gravityScale=0
    linearDampingOverride=0.995
    linearSpringConstant=155
    simulationSpace=BoneRelative
    alphaBoolBlend {blendOption=Linear}
    boundBone {boneName=dyn_cube_l_tip_04}
    componentAppliedLinearAccClamp=<6000 6000 6000>
    componentLinearVelScale=<0.9 0.9 0.9>
    physicsBodyDefinitions [{
      sphereCollisionRadius=0
      boundBone {boneName=dyn_cube_l_tip_04}
      boxExtents=<50 50 50>
      constraintSetup {angularLimitsMax=<0 0 0>, angularLimitsMin=<0 0 0>, angularTarget=<0 0 0>, linearAxesMax=<2 8 8>, linearAxesMin=<-8 -3 -8>}
      localJointOffset=<3 0 3>
     }]
    relativeSpaceBone {boneName=Melee_TwoHanded_Handle}
   }
   n36 = Anim Dynamics - Bone: dyn_cube_r_in_01(ComponentPose=@n35, Alpha=1, ExternalForce="0,0,0") {
    angularDampingOverride=0.9
    angularSpringConstant=175
    bAngularSpring=true
    bGravityOverrideInSimSpace=true
    bLinearSpring=true
    bOverrideAngularDamping=true
    bOverrideLinearDamping=true
    gravityScale=0
    linearDampingOverride=0.995
    linearSpringConstant=155
    simulationSpace=BoneRelative
    alphaBoolBlend {blendOption=Linear}
    boundBone {boneName=dyn_cube_r_in_01}
    componentAppliedLinearAccClamp=<6000 6000 6000>
    componentLinearVelScale=<0.9 0.9 0.9>
    physicsBodyDefinitions [{
      sphereCollisionRadius=0
      boundBone {boneName=dyn_cube_r_in_01}
      boxExtents=<15 15 15>
      constraintSetup {angularLimitsMax=<0 0 0>, angularLimitsMin=<0 0 0>, angularTarget=<0 0 0>, linearAxesMax=<7 8 8>, linearAxesMin=<-8 -2 -5>}
      localJointOffset=<3 3 3>
     }]
    relativeSpaceBone {boneName=dyn_master}
   }
   n37 = Anim Dynamics - Bone: dyn_cube_r_in_02(ComponentPose=@n36, Alpha=1, ExternalForce="0,0,0") {
    angularDampingOverride=0.9
    angularSpringConstant=195
    bAngularSpring=true
    bGravityOverrideInSimSpace=true
    bLinearSpring=true
    bOverrideAngularDamping=true
    bOverrideLinearDamping=true
    gravityScale=0
    linearDampingOverride=0.995
    linearSpringConstant=155
    simulationSpace=BoneRelative
    alphaBoolBlend {blendOption=Linear}
    boundBone {boneName=dyn_cube_r_in_02}
    componentAppliedLinearAccClamp=<6000 6000 6000>
    componentLinearVelScale=<0.9 0.9 0.9>
    physicsBodyDefinitions [{
      sphereCollisionRadius=0
      boundBone {boneName=dyn_cube_r_in_02}
      boxExtents=<15 15 15>
      constraintSetup {angularLimitsMax=<0 0 0>, angularLimitsMin=<0 0 0>, angularTarget=<0 0 0>, linearAxesMax=<8 3 8>, linearAxesMin=<-5 -8 -4>}
      localJointOffset=<-3 -3 -3>
     }]
    relativeSpaceBone {boneName=Melee_TwoHanded_Handle}
   }
   n38 = Anim Dynamics - Bone: dyn_cube_r_mid(ComponentPose=@n37, Alpha=1, ExternalForce="0,0,0") {
    angularDampingOverride=0.9
    angularSpringConstant=195
    bAngularSpring=true
    bGravityOverrideInSimSpace=true
    bLinearSpring=true
    bOverrideAngularDamping=true
    bOverrideLinearDamping=true
    gravityScale=0
    linearDampingOverride=0.995
    linearSpringConstant=155
    simulationSpace=BoneRelative
    alphaBoolBlend {blendOption=Linear}
    boundBone {boneName=dyn_cube_r_mid}
    componentAppliedLinearAccClamp=<6000 6000 6000>
    componentLinearVelScale=<0.9 0.9 0.9>
    physicsBodyDefinitions [{
      sphereCollisionRadius=0
      boundBone {boneName=dyn_cube_r_mid}
      boxExtents=<15 15 15>
      constraintSetup {angularLimitsMax=<0 0 0>, angularLimitsMin=<0 0 0>, angularTarget=<0 0 0>, linearAxesMax=<8 8 6>, linearAxesMin=<-8 -2 -3>}
      localJointOffset=<3 3 3>
     }]
    relativeSpaceBone {boneName=Melee_TwoHanded_Handle}
   }
   n39 = Anim Dynamics - Bone: dyn_cube_r_tip_01(ComponentPose=@n38, Alpha=1, ExternalForce="0,0,0") {
    angularDampingOverride=0.9
    angularSpringConstant=175
    bAngularSpring=true
    bGravityOverrideInSimSpace=true
    bLinearSpring=true
    bOverrideAngularDamping=true
    bOverrideLinearDamping=true
    gravityScale=0
    linearDampingOverride=0.995
    linearSpringConstant=155
    simulationSpace=BoneRelative
    alphaBoolBlend {blendOption=Linear}
    boundBone {boneName=dyn_cube_r_tip_01}
    componentAppliedLinearAccClamp=<6000 6000 6000>
    componentLinearVelScale=<0.9 0.9 0.9>
    physicsBodyDefinitions [{
      sphereCollisionRadius=0
      boundBone {boneName=dyn_cube_r_tip_01}
      boxExtents=<50 50 50>
      constraintSetup {angularLimitsMax=<0 0 0>, angularLimitsMin=<0 0 0>, angularTarget=<0 0 0>, linearAxesMax=<2 3 8>, linearAxesMin=<-8 -8 -2>}
      localJointOffset=<-1 -1 -1>
     }]
    relativeSpaceBone {boneName=Melee_TwoHanded_Handle}
   }
   n40 = Anim Dynamics - Bone: dyn_cube_r_tip_02(ComponentPose=@n39, Alpha=1, ExternalForce="0,0,0") {
    angularDampingOverride=0.9
    angularSpringConstant=175
    bAngularSpring=true
    bGravityOverrideInSimSpace=true
    bLinearSpring=true
    bOverrideAngularDamping=true
    bOverrideLinearDamping=true
    gravityScale=0
    linearDampingOverride=0.995
    linearSpringConstant=155
    simulationSpace=BoneRelative
    alphaBoolBlend {blendOption=Linear}
    boundBone {boneName=dyn_cube_r_tip_02}
    componentAppliedLinearAccClamp=<6000 6000 6000>
    componentLinearVelScale=<0.9 0.9 0.9>
    physicsBodyDefinitions [{
      sphereCollisionRadius=0
      boundBone {boneName=dyn_cube_r_tip_02}
      boxExtents=<50 50 50>
      constraintSetup {angularLimitsMax=<0 0 0>, angularLimitsMin=<0 0 0>, angularTarget=<0 0 0>, linearAxesMax=<2 3 8>, linearAxesMin=<-8 -8 -2>}
      localJointOffset=<3 3 3>
     }]
    relativeSpaceBone {boneName=Melee_TwoHanded_Handle}
   }
   n41 = Anim Dynamics - Bone: dyn_cube_r_tip_03(ComponentPose=@n40, Alpha=1, ExternalForce="0,0,0") {
    angularDampingOverride=0.9
    angularSpringConstant=175
    bAngularSpring=true
    bGravityOverrideInSimSpace=true
    bLinearSpring=true
    bOverrideAngularDamping=true
    bOverrideLinearDamping=true
    gravityScale=0
    linearDampingOverride=0.995
    linearSpringConstant=155
    simulationSpace=BoneRelative
    alphaBoolBlend {blendOption=Linear}
    boundBone {boneName=dyn_cube_r_tip_03}
    componentAppliedLinearAccClamp=<6000 6000 6000>
    componentLinearVelScale=<0.9 0.9 0.9>
    physicsBodyDefinitions [{
      sphereCollisionRadius=0
      boundBone {boneName=dyn_cube_r_tip_03}
      boxExtents=<50 50 50>
      constraintSetup {angularLimitsMax=<0 0 0>, angularLimitsMin=<0 0 0>, angularTarget=<0 0 0>, linearAxesMax=<2 3 8>, linearAxesMin=<-8 -8 -2>}
      localJointOffset=<0 0 0>
     }]
    relativeSpaceBone {boneName=Melee_TwoHanded_Handle}
   }
   Output Pose(Result=Two Way Blend(A=Component To Local(ComponentPose=@n25), B=Component To Local(ComponentPose=@n41), Alpha=Get Is_Swinging() -> (Is_Swinging:double)) {blendNode {alphaBoolBlend {blendOption=Linear}}}) {layerGroup=None}
   n42 = Get Floating_Rotator2() -> (Floating_Rotator2:Rotator)
   n43 = Get Cube_Float_A() -> (Cube_Float_A:Vector)
   n44 = "Transform (Modify) Bone - Bone: dyn_cube_l_in"(ComponentPose=Local To Component(LocalPose=Input Pose()), Alpha=0.2, Translation=@n43, Rotation=@n42, Scale="1,1,1") {rotationMode=BMM_Additive, translationMode=BMM_Additive, alphaBoolBlend {blendOption=Linear}, boneToModify {boneName=dyn_cube_l_in}}
   n45 = Get Floating_Rotator1() -> (Floating_Rotator1:Rotator)
   n46 = Get Cube_Float_B() -> (Cube_Float_B:Vector)
   n47 = "Transform (Modify) Bone - Bone: dyn_cube_l_mid"(ComponentPose=@n44, Alpha=0.4, Translation=@n46, Rotation=@n45, Scale="1,1,1") {rotationMode=BMM_Additive, translationMode=BMM_Additive, alphaBoolBlend {blendOption=Linear}, boneToModify {boneName=dyn_cube_l_mid}}
   n48 = "Transform (Modify) Bone - Bone: dyn_cube_l_out_01"(ComponentPose=@n47, Alpha=0.6, Translation=@n43, Rotation=@n42, Scale="1,1,1") {rotationMode=BMM_Additive, translationMode=BMM_Additive, alphaBoolBlend {blendOption=Linear}, boneToModify {boneName=dyn_cube_l_out_01}}
   n49 = Get Cube_Float_D() -> (Cube_Float_D:Vector)
   n50 = "Transform (Modify) Bone - Bone: dyn_cube_l_out_02"(ComponentPose=@n48, Alpha=0.5, Translation=@n49, Rotation=@n45, Scale="1,1,1") {rotationMode=BMM_Additive, translationMode=BMM_Additive, alphaBoolBlend {blendOption=Linear}, boneToModify {boneName=dyn_cube_l_out_02}}
   n51 = Get Cube_Float_C() -> (Cube_Float_C:Vector)
   n52 = "Transform (Modify) Bone - Bone: dyn_cube_l_out_03"(ComponentPose=@n50, Alpha=0.45, Translation=@n51, Rotation=@n42, Scale="1,1,1") {rotationMode=BMM_Additive, translationMode=BMM_Additive, alphaBoolBlend {blendOption=Linear}, boneToModify {boneName=dyn_cube_l_out_03}}
   n53 = "Transform (Modify) Bone - Bone: dyn_cube_l_out_04"(ComponentPose=@n52, Alpha=0.8, Translation=@n43, Rotation=@n45, Scale="1,1,1") {rotationMode=BMM_Additive, translationMode=BMM_Additive, alphaBoolBlend {blendOption=Linear}, boneToModify {boneName=dyn_cube_l_out_04}}
   n54 = "Transform (Modify) Bone - Bone: dyn_cube_l_tip_01"(ComponentPose=@n53, Alpha=0.7, Translation=@n43, Rotation=@n45, Scale="1,1,1") {rotationMode=BMM_Additive, translationMode=BMM_Additive, alphaBoolBlend {blendOption=Linear}, boneToModify {boneName=dyn_cube_l_tip_01}}
   n55 = "Transform (Modify) Bone - Bone: dyn_cube_l_tip_02"(ComponentPose=@n54, Alpha=0.75, Translation=@n51, Rotation=@n42, Scale="1,1,1") {rotationMode=BMM_Additive, translationMode=BMM_Additive, alphaBoolBlend {blendOption=Linear}, boneToModify {boneName=dyn_cube_l_tip_02}}
   n56 = "Transform (Modify) Bone - Bone: dyn_cube_l_tip_03"(ComponentPose=@n55, Alpha=0.85, Translation=@n49, Rotation=@n45, Scale="1,1,1") {rotationMode=BMM_Additive, translationMode=BMM_Additive, alphaBoolBlend {blendOption=Linear}, boneToModify {boneName=dyn_cube_l_tip_03}}
   n57 = "Transform (Modify) Bone - Bone: dyn_cube_l_tip_04"(ComponentPose=@n56, Alpha=0.85, Translation=@n46, Rotation=@n42, Scale="1,1,1") {rotationMode=BMM_Additive, translationMode=BMM_Additive, alphaBoolBlend {blendOption=Linear}, boneToModify {boneName=dyn_cube_l_tip_04}}
   n58 = "Transform (Modify) Bone - Bone: dyn_cube_r_in_01"(ComponentPose=@n57, Alpha=0.25, Translation=@n43, Rotation=@n42, Scale="1,1,1") {rotationMode=BMM_Additive, translationMode=BMM_Additive, alphaBoolBlend {blendOption=Linear}, boneToModify {boneName=dyn_cube_r_in_01}}
   n59 = "Transform (Modify) Bone - Bone: dyn_cube_r_in_02"(ComponentPose=@n58, Alpha=0.2, Translation=@n46, Rotation=@n45, Scale="1,1,1") {rotationMode=BMM_Additive, translationMode=BMM_Additive, alphaBoolBlend {blendOption=Linear}, boneToModify {boneName=dyn_cube_r_in_02}}
   n60 = "Transform (Modify) Bone - Bone: dyn_cube_r_mid"(ComponentPose=@n59, Alpha=0.4, Translation=@n51, Rotation=@n42, Scale="1,1,1") {rotationMode=BMM_Additive, translationMode=BMM_Additive, alphaBoolBlend {blendOption=Linear}, boneToModify {boneName=dyn_cube_r_mid}}
   n61 = "Transform (Modify) Bone - Bone: dyn_cube_r_tip_01"(ComponentPose=@n60, Alpha=0.8, Translation=@n49, Rotation=@n45, Scale="1,1,1") {rotationMode=BMM_Additive, translationMode=BMM_Additive, alphaBoolBlend {blendOption=Linear}, boneToModify {boneName=dyn_cube_r_tip_01}}
   n62 = "Transform (Modify) Bone - Bone: dyn_cube_r_tip_02"(ComponentPose=@n61, Alpha=0.75, Translation=@n46, Rotation=@n42, Scale="1,1,1") {rotationMode=BMM_Additive, translationMode=BMM_Additive, alphaBoolBlend {blendOption=Linear}, boneToModify {boneName=dyn_cube_r_tip_02}}
   n63 = "Transform (Modify) Bone - Bone: dyn_cube_r_tip_03"(ComponentPose=@n62, Alpha=0.9, Translation=@n43, Rotation=@n45, Scale="1,1,1") {rotationMode=BMM_Additive, translationMode=BMM_Additive, alphaBoolBlend {blendOption=Linear}, boneToModify {boneName=dyn_cube_r_tip_03}}
   Save cached pose 'CubesFloat'(Pose=Component To Local(ComponentPose=@n63)) {cacheName=CubesFloat}
  ]}
 EventGraph {nodes [
   Event BlueprintUpdateAnimation()
   n1 = GetOwnerSkeletalMeshComponent()
   > n2 = Random Vector Function!(
    execute
    CharacterPartAnimInstance:Custom Character Part Anim Instance Object Reference
    SkelMeshComponent:Mesh Component Object Reference=@n1
    JointName:Name=None
    GameTimeInSeconds:double=GetGameTimeInSeconds()
    TimeOffset:double=0.2
    PerlinScale:double=0
    Frequency:Vector="0.9,1.1,0.75"
    Amplitude:Vector="0.4,1.5,1.2"
    Offset:Vector="0,0,1"
   ) -> (then, VectorOut:Vector)
   > Set Cube_Float_A(Cube_Float_A:Vector=@n2) -> (Output_Get:Vector)
   > n3 = Random Vector Function(
    SkelMeshComponent=@n1
    JointName=None
    GameTimeInSeconds=GetGameTimeInSeconds()
    TimeOffset=3
    PerlinScale=0
    Frequency="0.8,0.5,0.45"
    Amplitude="0.6,1.8,1.4"
    Offset="0,0,0"
   )
   > Set Cube_Float_B(Cube_Float_B:Vector=@n3) -> (Output_Get:Vector)
   > n4 = Random Vector Function(
    SkelMeshComponent=@n1
    JointName=None
    GameTimeInSeconds=GetGameTimeInSeconds()
    TimeOffset=-2
    PerlinScale=0
    Frequency="0.9,1.35,0.75"
    Amplitude="0.3,1.5,1.2"
    Offset="0,0,0"
   )
   > Set Cube_Float_C(Cube_Float_C:Vector=@n4) -> (Output_Get:Vector)
   n5 = BreakVector(InVec=vector * float(A=@n4, B=17))
   > Set Floating_Rotator1(Floating_Rotator1:Rotator=MakeRotator(Roll=@n5, Pitch=@n5.Y, Yaw=@n5.Z)) -> (Output_Get:Rotator)
   > n6 = Random Vector Function(
    SkelMeshComponent=@n1
    JointName=None
    GameTimeInSeconds=GetGameTimeInSeconds()
    TimeOffset=1.1
    PerlinScale=0
    Frequency="0.8,1.4,0.7"
    Amplitude="0.6,1.8,1.1"
    Offset="0,0,0"
   )
   > Set Cube_Float_D(Cube_Float_D:Vector=@n6) -> (Output_Get:Vector)
   n7 = BreakVector(InVec=vector * float(A=@n6, B=15))
   > Set Floating_Rotator2(Floating_Rotator2:Rotator=MakeRotator(Roll=@n7, Pitch=@n7.Y, Yaw=@n7.Z)) -> (Output_Get:Rotator)
   > n8 = Branch(Condition=Get bIsPlayingMeleeAnim())
   > Set Is_Swinging(Is_Swinging:double=1) -> (Output_Get:double)
   Set Is_Swinging(execute=@n8.else, Is_Swinging=0)
  ]}
]
variables [
 bIsOnGround_0 {type=boolean}
 bIsPlayingMeleeAnim_0 {type=boolean}
 Cube_Float_A {type=Vector}
 Cube_Float_B {type=Vector}
 Cube_Float_C {type=Vector}
 Cube_Float_D {type=Vector}
 Floating_Rotator1 {type=Rotator}
 Floating_Rotator2 {type=Rotator}
 Is_Swinging {type=number}
 Jump {type=number}
]
