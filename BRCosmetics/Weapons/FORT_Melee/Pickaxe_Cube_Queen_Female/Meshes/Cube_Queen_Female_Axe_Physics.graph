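# Physics asset for the Cube Queen Female axe mesh, a cosmetic melee pickaxe.
# A single body on the dyn_master bone uses one long, thin capsule as a simplified collision shape, offset and rotated from the bone origin.
# With no other bodies and no constraints, the asset defines no articulated or jointed parts.
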PhysicsAsset /BRCosmetics/Weapons/FORT_Melee/Pickaxe_Cube_Queen_Female/Meshes/Cube_Queen_Female_Axe_Physics
bodies {dyn_master [{shapeType=Capsule, length=166.5, radius=5.005, shapeName=dyn_master_capsule, center=<0.8138 -0.0003 -54.1124>, rotation=<rot 0 -90 0>}]}
constraints []
